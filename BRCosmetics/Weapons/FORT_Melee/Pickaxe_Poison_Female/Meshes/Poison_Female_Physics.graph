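# Physics asset for the Poison Female pickaxe, a two-handed melee cosmetic weapon with a dangling chain.
# A long capsule covers the handle, and three capsules cover the dyn_chains bones, the last one larger than the other two.
# Constraints link the chain in series from the handle, with asymmetric swing and tighter twist at the root and looser, symmetric swing on the two lower joints.
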PhysicsAsset /BRCosmetics/Weapons/FORT_Melee/Pickaxe_Poison_Female/Meshes/Poison_Female_Physics
bodies {
 Melee_TwoHanded_Handle [{shapeType=Capsule, length=53.0387, radius=2.7934, center=<0.4361 -0.099 3.7548>, rotation=<rot 0.3805 -90.0014 0.087>}]
 dyn_chains_01 [{shapeType=Capsule, length=3.9543, radius=2.208, center=<1.5119 0 0>, rotation=<rot 0 -90 -90>}]
 dyn_chains_02 [{shapeType=Capsule, length=3.9543, radius=2.208, center=<2.4647 0 0>, rotation=<rot 0 -90 -90>}]
 dyn_chains_03 [{shapeType=Capsule, length=7.5404, radius=3.9243, center=<6.1978 -0.0039 0.0001>, rotation=<rot 0 -89.8557 -89.9251>}]
}
constraints [
 {bone1Name=dyn_chains_01, bone2Name=Melee_TwoHanded_Handle, swing1LimitDegrees=25, swing2LimitDegrees=90, twistLimitDegrees=10}
 {bone1Name=dyn_chains_02, bone2Name=dyn_chains_01, swing1LimitDegrees=75, swing2LimitDegrees=75, twistLimitDegrees=20}
 {bone1Name=dyn_chains_03, bone2Name=dyn_chains_02, swing1LimitDegrees=75, swing2LimitDegrees=75, twistLimitDegrees=20}
]
